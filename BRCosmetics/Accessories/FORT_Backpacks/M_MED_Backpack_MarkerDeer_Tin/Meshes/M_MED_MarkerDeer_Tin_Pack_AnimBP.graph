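# Animation Blueprint for the MarkerDeer Tin backpack that poses its ship body and wings with a chain of additive bone-space Modify Bone nodes.
# Each update derives body rotation and translation from the owner's spine_05-to-spine_01 delta from ref pose.
# That offset blends out while openWings is set by sprinting, tactical sprinting or skydive diving.
# Diving adds a body tilt, while fixed body and wing offsets stay enabled.
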AnimBlueprint /BRCosmetics/Accessories/FORT_Backpacks/M_MED_Backpack_MarkerDeer_Tin/Meshes/M_MED_MarkerDeer_Tin_Pack_AnimBP
target_skeleton=&/M_MED_MarkerDeer_Tin_Pack_Skeleton
graphs [
 AnimGraph {nodes [
   n1 = "Transform (Modify) Bone - Bone: base_shipBody"(ComponentPose=Local To Component(LocalPose=Input Pose()), Alpha=1, Translation="-3,0,0", Rotation="7,-7,0", Scale="1,1,1") {
    rotationMode=BMM_Additive
    rotationSpace=BCS_BoneSpace
    translationMode=BMM_Additive
    translationSpace=BCS_BoneSpace
    alphaBoolBlend {blendInTime=0.3, blendOption=Linear, blendOutTime=0.3}
    boneToModify {boneName=base_shipBody}
    rotation {pitch=7, yaw=-7}
    translation {x=-3}
   }
   n2 = "Transform (Modify) Bone - Bone: base_shipBody"!(
    ComponentPose=@n1
    bAlphaBoolEnabled:bool=NOT Boolean(A=Get openWings() -> (openWings:bool))
    Translation=Get backpack_translation()
    Rotation=Get backpack_rotate()
    Scale="1,1,1"
   ) -> (Pose) {
    alphaInputType=Bool
    rotationMode=BMM_Additive
    rotationSpace=BCS_BoneSpace
    translationMode=BMM_Additive
    translationSpace=BCS_BoneSpace
    alphaBoolBlend {blendInTime=0.3, blendOption=Linear, blendOutTime=0.1}
    boneToModify {boneName=base_shipBody}
   }
   n3 = "Transform (Modify) Bone - Bone: base_shipBody"(ComponentPose=@n2, bAlphaBoolEnabled=True, Translation="2.3,0,0", Rotation="5,-10,0", Scale="1,1,1") {
    alphaInputType=Bool
    rotationMode=BMM_Additive
    rotationSpace=BCS_BoneSpace
    translationMode=BMM_Additive
    translationSpace=BCS_BoneSpace
    alphaBoolBlend {blendInTime=0.5, blendOption=Sinusoidal, blendOutTime=0.5}
    boneToModify {boneName=base_shipBody}
    rotation {pitch=5, yaw=-10}
    translation {x=2.3}
   }
   n4 = "Transform (Modify) Bone - Bone: base_shipBody"(ComponentPose=@n3, bAlphaBoolEnabled=True, Translation="2.3,0,0", Rotation="5,0,0", Scale="1,1,1") {
    alphaInputType=Bool
    rotationMode=BMM_Additive
    rotationSpace=BCS_BoneSpace
    translationMode=BMM_Additive
    translationSpace=BCS_BoneSpace
    alphaBoolBlend {blendInTime=0.5, blendOption=Sinusoidal, blendOutTime=0.5}
    boneToModify {boneName=base_shipBody}
    rotation {pitch=5}
    translation {x=2.3}
   }
   n5 = "Transform (Modify) Bone - Bone: base_shipBody"(ComponentPose=@n4, bAlphaBoolEnabled=Get bIsSkydiveDiveMode(), Translation="2,0,0", Rotation="0,0,6", Scale="1,1,1") {
    alphaInputType=Bool
    rotationMode=BMM_Additive
    rotationSpace=BCS_BoneSpace
    translationMode=BMM_Additive
    translationSpace=BCS_BoneSpace
    alphaBoolBlend {blendInTime=0.3, blendOption=Linear, blendOutTime=3}
    boneToModify {boneName=base_shipBody}
    rotation {roll=6}
    translation {x=2}
   }
   n6 = "Transform (Modify) Bone - Bone: dyn_topWing_r"!(ComponentPose=@n5, bAlphaBoolEnabled:bool=True, Translation="0,0,0", Rotation="0,15,0", Scale="1,1,1") -> (Pose) {
    alphaInputType=Bool
    rotationMode=BMM_Additive
    rotationSpace=BCS_BoneSpace
    alphaBoolBlend {blendInTime=0.6, blendOption=Linear, blendOutTime=0.3}
    boneToModify {boneName=dyn_topWing_r}
    rotation {yaw=15}
   }
   n7 = "Transform (Modify) Bone - Bone: dyn_bottomWing_r"!(ComponentPose=@n6, bAlphaBoolEnabled:bool=True, Translation="0,0,0", Rotation="0,-5,0", Scale="1,1,1") -> (Pose) {
    alphaInputType=Bool
    rotationMode=BMM_Additive
    rotationSpace=BCS_BoneSpace
    alphaBoolBlend {blendInTime=0.6, blendOption=Linear, blendOutTime=0.3}
    boneToModify {boneName=dyn_bottomWing_r}
    rotation {yaw=-5}
   }
   n8 = "Transform (Modify) Bone - Bone: dyn_topWing_l"!(ComponentPose=@n7, bAlphaBoolEnabled:bool=True, Translation="0,0,0", Rotation="0,-15,0", Scale="1,1,1") -> (Pose) {
    alphaInputType=Bool
    rotationMode=BMM_Additive
    rotationSpace=BCS_BoneSpace
    alphaBoolBlend {blendInTime=0.6, blendOption=Linear, blendOutTime=0.3}
    boneToModify {boneName=dyn_topWing_l}
    rotation {yaw=-15}
   }
   n9 = "Transform (Modify) Bone - Bone: dyn_bottomWing_l"!(ComponentPose=@n8, bAlphaBoolEnabled:bool=True, Translation="0,0,0", Rotation="0,5,0", Scale="1,1,1") -> (Pose) {
    alphaInputType=Bool
    rotationMode=BMM_Additive
    rotationSpace=BCS_BoneSpace
    alphaBoolBlend {blendInTime=0.6, blendOption=Linear, blendOutTime=0.3}
    boneToModify {boneName=dyn_bottomWing_l}
    rotation {yaw=5}
   }
   Output Pose(Result=Component To Local(ComponentPose=@n9)) {layerGroup=None}
  ]}
 EventGraph {nodes [
   Event BlueprintUpdateAnimation()
   > n1 = Sequence()
   > n2 = Branch(Condition=OR Boolean!(A=Get bIsTacticalSprinting(), B=Get bIsSprinting(), C:bool=Get bIsSkydiveDiveMode()) -> (ReturnValue))
   > Set openWings(openWings:bool=true) -> (Output_Get:bool)
   Set openWings(execute=@n2.else, openWings=false)
   n3 = GetOwnerSkeletalMeshComponent()
   n4 = BreakRotator(InRot=BreakTransform(InTransform=GetDeltaTransformFromRefPose(self=@n3, BoneName=spine_05, BaseName=spine_01)).Rotation)
   Set backpack_rotate(execute=@n1.then_1, backpack_rotate=MakeRotator(Yaw=MapRangeClamped(Value=@n4, InRangeA=0, InRangeB=-45, OutRangeB=-10)))
   n5 = MapRangeClamped(Value=@n4, InRangeA=-55, InRangeB=45, OutRangeA=-2, OutRangeB=3)
   n6 = MapRangeClamped(Value=@n4.Yaw, InRangeA=23, InRangeB=35, OutRangeB=2.3)
   > Set backpack_translation(backpack_translation=MakeVector(X=float + float(A=@n6, B=@n5), Y=0))
   TryGetPawnOwner()
  ]}
]
variables [
 backpack_rotate {type=Rotator}
 backpack_translation {type=Vector}
 openWings {type=boolean}
]
